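# Conversation database for the Warty Brine NPC, serving as the top-level dialogue hub when a player talks to this character.
# One entry point leads to a greeting Speech node whose contextual lines react to first-ever meetings, per-match conversation counts and two specific outfits, using curve-driven priorities.
# That node fans out through link nodes to remote entries for quests, basic services, Warty Brine dialogue, seasonal story, Dual Wield expertise and character stories.
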ConversationDatabase /NPCLibrary/NPCs/WartyBrine/NPCLibrary_Conversation_WartyBrine
entry_points [{tag=Conversation.Entry.WartyBrine, destinations [#1]}]
nodes [
 ConversationEntryPointNode_7 {
  id=#1
  class=ConversationEntryPointNode
  outputs [#2]
  properties {entryTag {tagName=Conversation.Entry.WartyBrine}}
 }
 Speech_C_4 {
  id=#2
  class=Speech_C
  outputs [
   #3
   #4
   #5
   #6
   #7
   #8
   #9
   #10
   #11
   #12
   #13
  ]
  properties {
   nodeName=""
   generalConfig {
    defaultMessage=Stay salty.
    contextualMessages [
     {
      message=Don't act like you don't know who I am.
      contextRequirements [{requirement {class=FortControllerRequirement_IsFirstEverConversationWithNPC}}]
      priority {curve {rowName=Tandem.Speech.Priority.FirstEverInteraction}}
     }
     {
      message=Stay salty.
      contextRequirements [{requirement {class=FortControllerRequirement_CheckNPCMatchConversationCount, properties {nPCUniqueIdTag {tagName=AISpawnerData.Type.Tandem.WartyBrine}}}}]
      priority {curve {rowName=None}}
     }
     {
      message=Real ones stay bad to the brine.
      contextRequirements [{requirement {class=FortControllerRequirement_CheckNPCMatchConversationCount, properties {nPCUniqueIdTag {tagName=AISpawnerData.Type.Tandem.WartyBrine}}}}]
      priority {curve {rowName=None}}
     }
     {
      message="Why the sour face? Ha!"
      contextRequirements [{requirement {class=FortControllerRequirement_CheckNPCMatchConversationCount, properties {nPCUniqueIdTag {tagName=AISpawnerData.Type.Tandem.WartyBrine}}}}]
      priority {curve {rowName=None}}
     }
     {
      message="Tasty Bois for life! Or till I get an offer for a solo album, at least."
      contextRequirements [{requirement {
         class=FortControllerRequirement_HasCID
         properties {allowedCIds [/BRCosmetics/Athena/Items/Cosmetics/Characters/CID_614_Athena_Commando_M_MissingLink]}
        }}]
      priority {curve {rowName=Tandem.Speech.Priority.OutfitReaction}}
     }
     {
      message="Step off my spot, alpaca head."
      contextRequirements [{requirement {
         class=FortControllerRequirement_HasCID
         properties {allowedCIds [/BRCosmetics/Athena/Items/Cosmetics/Characters/CID_233_Athena_Commando_M_FortniteDJ]}
        }}]
      priority {curve {rowName=Tandem.Speech.Priority.OutfitReaction}}
     }
    ]
   }
  }
 }
 ConversationLinkNode_10 {
  id=#13
  class=ConversationLinkNode
  properties {remoteEntryTag {tagName=Conversation.Entry.Season36.DJYonder.CharacterStories.01}}
 }
 ConversationLinkNode_9 {
  id=#12
  class=ConversationLinkNode
  properties {remoteEntryTag {tagName=Conversation.Entry.Season36.BigDill.CharacterStories.01}}
 }
 ConversationLinkNode_8 {
  id=#11
  class=ConversationLinkNode
  properties {remoteEntryTag {tagName=Conversation.Entry.Expertise.DualWield}}
 }
 ConversationLinkNode_7 {
  id=#10
  class=ConversationLinkNode
  properties {remoteEntryTag {tagName=Conversation.Entry.Season34.Story.P02.Q03C}}
 }
 ConversationLinkNode_6 {
  id=#9
  class=ConversationLinkNode
  properties {remoteEntryTag {tagName=Conversation.Entry.WartyBrine.Dialogue}}
 }
 ConversationLinkNode_5 {
  id=#8
  class=ConversationLinkNode
  properties {remoteEntryTag {tagName=Conversation.Entry.Season34.Story.P02.Q03B}}
 }
 ConversationLinkNode_3 {
  id=#7
  class=ConversationLinkNode
  properties {remoteEntryTag {tagName=Conversation.Entry.Season34.Expertise.DualWield.Introduction}}
 }
 ConversationLinkNode_4 {
  id=#6
  class=ConversationLinkNode
  properties {remoteEntryTag {tagName=Conversation.Entry.Season34.Expertise.DualWield.Finish}}
 }
 ConversationLinkNode_2 {
  id=#5
  class=ConversationLinkNode
  properties {remoteEntryTag {tagName=Conversation.Entry.Season34.Story.P02.Q03}}
 }
 ConversationLinkNode_0 {id=#4, class=ConversationLinkNode, properties {remoteEntryTag {tagName=Conversation.Entry.BasicServices}}}
 ConversationLinkNode_1 {id=#3, class=ConversationLinkNode, properties {remoteEntryTag {tagName=Conversation.Entry.NPCLibrary.Quests}}}
]
