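# Conversation database for the NPC WartyBrine's "Ask about..." dialogue, where the pickle musician talks about their music career.
# An entry tag leads to a greeting speech offering choices for Tasty Bois, New Album and Debut Album, plus a Back option.
# Each topic plays a linear chain of speech lines, and all three converge on one shared ReturnToStart node labelled BACK.
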ConversationDatabase /NPCLibrary/NPCs/WartyBrine/NPCLibrary_Conversation_WartyBrine_Dialogue
entry_points [{tag=Conversation.Entry.WartyBrine.Dialogue, destinations [#1]}]
nodes [
 ConversationEntryPointNode_0 {
  id=#1
  class=ConversationEntryPointNode
  outputs [#2]
  properties {entryTag {tagName=Conversation.Entry.WartyBrine.Dialogue}}
 }
 Speech_C_0 {
  id=#2
  class=Speech_C
  outputs [#3, #4, #5, #6]
  properties {
   nodeName=""
   softTaskIcon=&T_UI_NPC_question
   generalConfig {defaultMessage=I can be sweet or I can be sour. Ask carefully.}
  }
  sub_nodes [ConversationChoiceNode_5 {class=ConversationChoiceNode, properties {defaultChoiceDisplayText=Ask about...}}]
 }
 Speech_C_4 {
  id=#6
  class=Speech_C
  outputs [#7]
  properties {
   nodeName=""
   softTaskIcon=&T_UI_NPC_question
   generalConfig {defaultMessage="If you've heard my debut country album, Dillinger the Gherk... no you didn't."}
  }
  sub_nodes [ConversationChoiceNode_2 {class=ConversationChoiceNode, properties {defaultChoiceDisplayText=Debut Album}}]
 }
 Speech_C_5 {
  id=#7
  class=Speech_C
  outputs [#8]
  properties {
   nodeName=""
   softTaskIcon=&T_UI_NPC_question
   generalConfig {defaultMessage=And if you own a copy of Dillinger the Gherk... Destroy it.}
  }
 }
 Speech_C_6 {
  id=#8
  class=Speech_C
  outputs [#9]
  properties {
   nodeName=""
   softTaskIcon=&T_UI_NPC_question
   generalConfig {defaultMessage=There's a couple tracks on there that would get this pickle put right back in the jar.}
  }
 }
 ReturnToStart_C_23 {
  id=#9
  class=ReturnToStart_C
  properties {softTaskIcon=&T_UI_NPC_Back}
  sub_nodes [ConversationChoiceNode_32 {class=ConversationChoiceNode, properties {defaultChoiceDisplayText=BACK}}]
 }
 Speech_C_1 {
  id=#5
  class=Speech_C
  outputs [#10]
  properties {
   nodeName=""
   softTaskIcon=&T_UI_NPC_question
   generalConfig {defaultMessage=I made a mint and a half off my... previous investment.}
  }
  sub_nodes [ConversationChoiceNode_1 {class=ConversationChoiceNode, properties {defaultChoiceDisplayText=New Album}}]
 }
 Speech_C_2 {
  id=#10
  class=Speech_C
  outputs [#11]
  properties {
   nodeName=""
   softTaskIcon=&T_UI_NPC_question
   generalConfig {defaultMessage="Funnelling those funds into a super secret solo album. Calling it Pump && Dump."}
  }
 }
 Speech_C_3 {
  id=#11
  class=Speech_C
  outputs [#12]
  properties {nodeName="", softTaskIcon=&T_UI_NPC_question, generalConfig {defaultMessage="Is that... insensitive? "}}
 }
 Speech_C_13 {
  id=#12
  class=Speech_C
  outputs [#9]
  properties {
   nodeName=""
   softTaskIcon=&T_UI_NPC_question
   generalConfig {defaultMessage="Eh, whatever. If you can't handle the brine, go eat a cucumber."}
  }
 }
 Back_C_0 {
  id=#4
  class=Back_C
  sub_nodes [ConversationChoiceNode_7 {class=ConversationChoiceNode, properties {defaultChoiceDisplayText=BACK}}]
 }
 Speech_C_7 {
  id=#3
  class=Speech_C
  outputs [#13]
  properties {
   nodeName=""
   softTaskIcon=&T_UI_NPC_question
   generalConfig {defaultMessage="Tasty Bois will always mean something to me, for real."}
  }
  sub_nodes [ConversationChoiceNode_0 {class=ConversationChoiceNode, properties {defaultChoiceDisplayText=Tasty Bois}}]
 }
 Speech_C_15 {
  id=#13
  class=Speech_C
  outputs [#14]
  properties {
   nodeName=""
   softTaskIcon=&T_UI_NPC_question
   generalConfig {defaultMessage="I love my boi The Brat, but he ain't got much spice. And I'm too tasty to share the stage forever."}
  }
 }
 Speech_C_14 {
  id=#14
  class=Speech_C
  outputs [#9]
  properties {
   nodeName=""
   softTaskIcon=&T_UI_NPC_question
   generalConfig {defaultMessage=I'm a Tasty Boi for now... but ya boi Big Dill is almost ready for his solo spotlight.}
  }
 }
]
